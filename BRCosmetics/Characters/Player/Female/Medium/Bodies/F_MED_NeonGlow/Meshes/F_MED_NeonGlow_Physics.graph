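# Physics asset for the NeonGlow female medium player body, setting up physics for six dangling straps, three per side, around the pelvis and thighs.
# Capsules cover pelvis, thighs and straps, with thighs and tiny MSTR root bodies joined to the pelvis by angularly free constraints.
# Straps hang from their MSTR bodies with narrow swing and zero twist, except fully locked left strap 3.
# Left strap 2 alone forms a four-segment chain, its second joint locked.
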PhysicsAsset /BRCosmetics/Characters/Player/Female/Medium/Bodies/F_MED_NeonGlow/Meshes/F_MED_NeonGlow_Physics
bodies {
 dyn_strap_l1_01 [{
   shapeType=Capsule
   length=3.0891
   radius=2.3934
   shapeName=dyn_strap_l1_01_capsule
   center=<8.9087 -1.1506 0.4727>
   rotation=<rot 45.1211 0 0>
  }]
 dyn_strap_l1_01_MSTR [{shapeType=Capsule, length=0.505, radius=0.505, shapeName=dyn_strap_l1_01_MSTR_capsule, rotation=<rot 0 -90 0>}]
 dyn_strap_l2_01 [{shapeType=Capsule, length=4.569, radius=3.3158, shapeName=dyn_strap_l2_01_capsule, center=<17.6853 -2.5537 0>}]
 dyn_strap_l2_01_MSTR [{shapeType=Capsule, length=0.505, radius=0.505, shapeName=dyn_strap_l2_01_MSTR_capsule, rotation=<rot 0 -90 0>}]
 dyn_strap_l2_02 [{shapeType=Capsule, length=10.5186, radius=1.7486, shapeName=dyn_strap_l2_02_capsule}]
 dyn_strap_l2_03 [{shapeType=Capsule, length=11.1773, radius=1.3024, shapeName=dyn_strap_l2_03_capsule, center=<5.8882 0 0>, rotation=<rot 90 0 0>}]
 dyn_strap_l2_04 [{
   shapeType=Capsule
   length=14.4695
   radius=1.4423
   shapeName=dyn_strap_l2_04_capsule
   center=<7.9873 0 0>
   rotation=<rot 90 -177.4734 -179.5497>
  }]
 dyn_strap_l3_01 [{
   shapeType=Capsule
   length=4.5369
   radius=1.9889
   shapeName=dyn_strap_l3_01_capsule
   center=<12.634 -2.2073 0.0109>
   rotation=<rot 90 0 -14.5537>
  }]
 dyn_strap_l3_01_MSTR [{shapeType=Capsule, length=0.505, radius=0.505, shapeName=dyn_strap_l3_01_MSTR_capsule, rotation=<rot 0 -90 0>}]
 dyn_strap_r1_01 [{shapeType=Capsule, length=2.1727, radius=2.5825, shapeName=dyn_strap_r1_01_capsule, center=<9.1236 1.148 0>, rotation=<rot 38.2354 0 0>}]
 dyn_strap_r1_01_MSTR [{shapeType=Capsule, length=0.505, radius=0.505, shapeName=dyn_strap_r1_01_MSTR_capsule, rotation=<rot 0 -90 0>}]
 dyn_strap_r2_01 [{shapeType=Capsule, length=7.9975, radius=3.6275, shapeName=dyn_strap_r2_01_capsule, center=<15.8203 3.1047 0>}]
 dyn_strap_r2_01_MSTR [{shapeType=Capsule, length=0.505, radius=0.505, shapeName=dyn_strap_r2_01_MSTR_capsule, rotation=<rot 0 -90 0>}]
 dyn_strap_r3_01 [{shapeType=Capsule, length=4.3367, radius=1.8418, shapeName=dyn_strap_r3_01_capsule, rotation=<rot 90 0 0>}]
 dyn_strap_r3_01_MSTR [{shapeType=Capsule, length=0.505, radius=0.505, shapeName=dyn_strap_r3_01_MSTR_capsule, rotation=<rot 0 -90 0>}]
 pelvis [{shapeType=Capsule, length=11.0592, radius=7.7858, shapeName=pelvis_capsule}]
 thigh_l [{shapeType=Capsule, length=19.8567, radius=7.6274, shapeName=thigh_l_capsule, center=<-17.3266 0 0>, rotation=<rot 90 0 0>}]
 thigh_r [{shapeType=Capsule, length=19.8567, radius=7.6274, shapeName=thigh_r_capsule, center=<17.3266 0 0>, rotation=<rot 90 0 180>}]
}
constraints [
 {bone1Name=dyn_strap_l1_01, bone2Name=dyn_strap_l1_01_MSTR, swing1LimitDegrees=10, swing2LimitDegrees=0, twistLimitDegrees=0}
 {bone1Name=dyn_strap_l2_01, bone2Name=dyn_strap_l2_01_MSTR, swing1LimitDegrees=20, swing2LimitDegrees=20, twistLimitDegrees=0}
 {
  bone1Name=dyn_strap_l2_02
  bone2Name=dyn_strap_l2_01
  swing1LimitDegrees=45
  swing1Motion=Locked
  swing2LimitDegrees=45
  swing2Motion=Locked
  twistLimitDegrees=45
  twistMotion=Locked
 }
 {bone1Name=dyn_strap_l2_03, bone2Name=dyn_strap_l2_02, swing1LimitDegrees=40, swing2LimitDegrees=10, twistLimitDegrees=0}
 {bone1Name=dyn_strap_l2_04, bone2Name=dyn_strap_l2_03, swing1LimitDegrees=30, swing2LimitDegrees=10, twistLimitDegrees=0}
 {bone1Name=dyn_strap_r2_01, bone2Name=dyn_strap_r2_01_MSTR, swing1LimitDegrees=20, swing2LimitDegrees=10, twistLimitDegrees=0}
 {bone1Name=dyn_strap_r3_01, bone2Name=dyn_strap_r3_01_MSTR, swing1LimitDegrees=0, swing2LimitDegrees=10, twistLimitDegrees=0}
 {bone1Name=dyn_strap_r1_01, bone2Name=dyn_strap_r1_01_MSTR, swing1LimitDegrees=10, swing2LimitDegrees=0, twistLimitDegrees=0}
 {
  bone1Name=thigh_l
  bone2Name=pelvis
  swing1LimitDegrees=45
  swing1Motion=Free
  swing2LimitDegrees=45
  swing2Motion=Free
  twistLimitDegrees=45
  twistMotion=Free
 }
 {
  bone1Name=thigh_r
  bone2Name=pelvis
  swing1LimitDegrees=45
  swing1Motion=Free
  swing2LimitDegrees=45
  swing2Motion=Free
  twistLimitDegrees=45
  twistMotion=Free
 }
 {
  bone1Name=dyn_strap_l2_01_MSTR
  bone2Name=pelvis
  swing1LimitDegrees=45
  swing1Motion=Free
  swing2LimitDegrees=45
  swing2Motion=Free
  twistLimitDegrees=45
  twistMotion=Free
 }
 {
  bone1Name=dyn_strap_r2_01_MSTR
  bone2Name=pelvis
  swing1LimitDegrees=45
  swing1Motion=Free
  swing2LimitDegrees=45
  swing2Motion=Free
  twistLimitDegrees=45
  twistMotion=Free
 }
 {
  bone1Name=dyn_strap_l1_01_MSTR
  bone2Name=pelvis
  swing1LimitDegrees=45
  swing1Motion=Free
  swing2LimitDegrees=45
  swing2Motion=Free
  twistLimitDegrees=45
  twistMotion=Free
 }
 {
  bone1Name=dyn_strap_r1_01_MSTR
  bone2Name=pelvis
  swing1LimitDegrees=45
  swing1Motion=Free
  swing2LimitDegrees=45
  swing2Motion=Free
  twistLimitDegrees=45
  twistMotion=Free
 }
 {
  bone1Name=dyn_strap_r3_01_MSTR
  bone2Name=pelvis
  swing1LimitDegrees=45
  swing1Motion=Free
  swing2LimitDegrees=45
  swing2Motion=Free
  twistLimitDegrees=45
  twistMotion=Free
 }
 {
  bone1Name=dyn_strap_l3_01_MSTR
  bone2Name=pelvis
  swing1LimitDegrees=45
  swing1Motion=Free
  swing2LimitDegrees=45
  swing2Motion=Free
  twistLimitDegrees=45
  twistMotion=Free
 }
 {
  bone1Name=dyn_strap_l3_01
  bone2Name=dyn_strap_l3_01_MSTR
  swing1LimitDegrees=0
  swing1Motion=Locked
  swing2LimitDegrees=5
  swing2Motion=Locked
  twistLimitDegrees=0
  twistMotion=Locked
 }
]
