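# Post-copy Animation Blueprint for the female medium NeonGlow player body, animating its secondary bones: straps, spurs and buckle.
# It copies the attached parent's pose, applies bone constraints and a crouch-aware Control Rig, then caches the result.
# RigidBody, AnimDynamics and bone offsets simulate that pose before additive spur rotations feed the output.
# Separate Look At and long AnimDynamics strap chains, including a Zipline switch, sit outside this visible output path.
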AnimBlueprint /BRCosmetics/Characters/Player/Female/Medium/Bodies/F_MED_NeonGlow/Meshes/F_MED_NeonGlow_AnimBP
target_skeleton=&/F_MED_NeonGlow_Skeleton
graphs [
 AnimGraph {nodes [
   n1 = Anim Dynamics - Bone: dyn_strap_l2_03!(ComponentPose, bAlphaBoolEnabled:bool=True, ExternalForce="0,0,0") -> (Pose) {
    alphaInputType=Bool
    angularDampingOverride=1
    angularSpringConstant=1000
    bAngularSpring=true
    bOverrideAngularDamping=true
    bUseGravityOverride=true
    bUsePlanarLimit=false
    simulationSpace=RootRelative
    alphaBoolBlend {blendInTime=0.1, blendOption=Linear, blendOutTime=0.1}
    boundBone {boneName=dyn_strap_l2_03}
    componentAppliedLinearAccClamp=<3000 3000 3000>
    componentLinearAccScale=<0.4 0.5 0.5>
    componentLinearVelScale=<0.5 0.5 0.5>
    gravityOverride {z=-980}
    physicsBodyDefinitions [{
      collisionType=CustomSphere
      sphereCollisionRadius=2.5
      boundBone {boneName=dyn_strap_l2_03}
      boxExtents=<8 4 8>
      constraintSetup {angularLimitsMax=<15 15 15>, angularLimitsMin=<-15 -15 -25>, angularTarget=<0 0 0>}
      localJointOffset=<1 0 0>
     }]
    relativeSpaceBone {boneName=pelvis}
   }
   Anim Dynamics - Bone: dyn_strap_l2_04!(ComponentPose=@n1, bAlphaBoolEnabled:bool=True, ExternalForce="0,0,0") -> (Pose) {
    alphaInputType=Bool
    angularDampingOverride=1
    angularSpringConstant=1000
    bAngularSpring=true
    bOverrideAngularDamping=true
    bUseGravityOverride=true
    bUsePlanarLimit=false
    simulationSpace=RootRelative
    alphaBoolBlend {blendInTime=0.1, blendOption=Linear, blendOutTime=0.1}
    boundBone {boneName=dyn_strap_l2_04}
    componentAppliedLinearAccClamp=<3000 3000 3000>
    componentLinearAccScale=<0.4 0.5 0.5>
    componentLinearVelScale=<0.5 0.5 0.5>
    gravityOverride {z=-980}
    physicsBodyDefinitions [{
      collisionType=CustomSphere
      sphereCollisionRadius=2.5
      boundBone {boneName=dyn_strap_l2_04}
      boxExtents=<8 4 8>
      constraintSetup {angularLimitsMax=<15 15 15>, angularLimitsMin=<-15 -15 -25>, angularTarget=<0 0 0>}
      localJointOffset=<1 0 0>
     }]
    relativeSpaceBone {boneName=pelvis}
   }
   Component To Local()
   Copy Pose From Mesh() {bUseAttachedParent=true}
   n2 = Anim Dynamics - Bone: dyn_strap_l2_01!(ComponentPose=Local To Component(LocalPose=Use cached pose 'Presim'()), bAlphaBoolEnabled:bool=True, ExternalForce="0,0,0") -> (Pose) {
    alphaInputType=Bool
    angularDampingOverride=1
    angularSpringConstant=2000
    bAngularSpring=true
    bOverrideAngularDamping=true
    bUseGravityOverride=true
    bUseSphericalLimits=true
    simulationSpace=BoneRelative
    alphaBoolBlend {blendInTime=0.1, blendOption=Custom, blendOutTime=0.1}
    boundBone {boneName=dyn_strap_l2_01}
    componentAppliedLinearAccClamp=<1000 1000 1000>
    componentLinearAccScale=<0.3 0.3 0.3>
    componentLinearVelScale=<0.3 0.3 0.3>
    gravityOverride {z=-980}
    physicsBodyDefinitions [{
      boundBone {boneName=dyn_strap_l2_01}
      boxExtents=<20 10 20>
      constraintSetup {angularLimitsMax=<0 20 0>, angularLimitsMin=<0 -20 -25>, angularTarget=<0 0 0>}
      localJointOffset=<1 0 0>
     }]
    relativeSpaceBone {boneName=root}
    sphericalLimits [{limitRadius=8, drivingBone {boneName=thigh_l}, sphereLocalOffset=<0 0 0>}]
   }
   n3 = Anim Dynamics - Bone: dyn_strap_l1_01!(ComponentPose=@n2, bAlphaBoolEnabled:bool=True, ExternalForce="0,0,0") -> (Pose) {
    alphaInputType=Bool
    angularDampingOverride=1
    angularSpringConstant=2000
    bAngularSpring=true
    bOverrideAngularDamping=true
    bUseGravityOverride=true
    bUseSphericalLimits=true
    simulationSpace=BoneRelative
    alphaBoolBlend {blendInTime=0.1, blendOption=Custom, blendOutTime=0.1}
    boundBone {boneName=dyn_strap_l1_01}
    componentAppliedLinearAccClamp=<1000 1000 1000>
    componentLinearAccScale=<0.3 0.3 0.3>
    componentLinearVelScale=<0.3 0.3 0.3>
    gravityOverride {z=-980}
    physicsBodyDefinitions [{
      boundBone {boneName=dyn_strap_l1_01}
      boxExtents=<10 5 10>
      constraintSetup {angularLimitsMax=<0 10 0>, angularLimitsMin=<0 -10 -15>, angularTarget=<0 0 0>}
      localJointOffset=<1 0 0>
     }]
    relativeSpaceBone {boneName=root}
    sphericalLimits [{limitRadius=8, drivingBone {boneName=thigh_l}, sphereLocalOffset=<0 0 0>}]
   }
   n4 = Anim Dynamics - Bone: dyn_strap_l3_01!(ComponentPose=@n3, bAlphaBoolEnabled:bool=True, ExternalForce="0,0,0") -> (Pose) {
    alphaInputType=Bool
    angularDampingOverride=1
    angularSpringConstant=2000
    bAngularSpring=true
    bOverrideAngularDamping=true
    bUseGravityOverride=true
    bUseSphericalLimits=true
    simulationSpace=BoneRelative
    alphaBoolBlend {blendInTime=0.1, blendOption=Custom, blendOutTime=0.1}
    boundBone {boneName=dyn_strap_l3_01}
    componentAppliedLinearAccClamp=<1000 1000 1000>
    componentLinearAccScale=<0.3 0.3 0.3>
    componentLinearVelScale=<0.3 0.3 0.3>
    gravityOverride {z=-980}
    physicsBodyDefinitions [{
      boundBone {boneName=dyn_strap_l3_01}
      boxExtents=<10 5 10>
      constraintSetup {angularLimitsMax=<0 0 0>, angularLimitsMin=<0 -10 -15>, angularTarget=<0 0 0>}
      localJointOffset=<1 0 0>
     }]
    relativeSpaceBone {boneName=root}
    sphericalLimits [{limitRadius=8, drivingBone {boneName=thigh_l}, sphereLocalOffset=<0 0 0>}]
   }
   n5 = Anim Dynamics - Bone: dyn_strap_l3_02!(ComponentPose=@n4, bAlphaBoolEnabled:bool=True, ExternalForce="0,0,0") -> (Pose) {
    alphaInputType=Bool
    angularDampingOverride=1
    angularSpringConstant=2000
    bAngularSpring=true
    bOverrideAngularDamping=true
    bUseGravityOverride=true
    bUseSphericalLimits=true
    simulationSpace=BoneRelative
    alphaBoolBlend {blendInTime=0.1, blendOption=Custom, blendOutTime=0.1}
    boundBone {boneName=dyn_strap_l3_02}
    componentAppliedLinearAccClamp=<1000 1000 1000>
    componentLinearAccScale=<0.3 0.3 0.3>
    componentLinearVelScale=<0.3 0.3 0.3>
    gravityOverride {z=-980}
    physicsBodyDefinitions [{
      boundBone {boneName=dyn_strap_l3_02}
      boxExtents=<10 5 10>
      constraintSetup {angularLimitsMax=<0 0 0>, angularLimitsMin=<0 -10 -15>, angularTarget=<0 0 0>}
      localJointOffset=<1 0 0>
     }]
    relativeSpaceBone {boneName=root}
    sphericalLimits [{limitRadius=8, drivingBone {boneName=thigh_l}, sphereLocalOffset=<0 0 0>}]
   }
   n6 = Anim Dynamics - Bone: dyn_strap_l2_02!(ComponentPose=@n5, bAlphaBoolEnabled:bool=True, ExternalForce="0,0,0") -> (Pose) {
    alphaInputType=Bool
    angularDampingOverride=1
    angularSpringConstant=1000
    bAngularSpring=true
    bOverrideAngularDamping=true
    bUseGravityOverride=true
    bUsePlanarLimit=false
    simulationSpace=RootRelative
    alphaBoolBlend {blendInTime=0.1, blendOption=Linear, blendOutTime=0.1}
    boundBone {boneName=dyn_strap_l2_02}
    componentAppliedLinearAccClamp=<1000 1000 1000>
    componentLinearAccScale=<0.3 0.3 0.3>
    componentLinearVelScale=<0.3 0.3 0.3>
    gravityOverride {z=-980}
    physicsBodyDefinitions [{
      collisionType=CustomSphere
      sphereCollisionRadius=2.5
      boundBone {boneName=dyn_strap_l2_02}
      boxExtents=<8 4 8>
      constraintSetup {angularLimitsMax=<15 15 0>, angularLimitsMin=<-15 -15 -25>, angularTarget=<0 0 0>}
      localJointOffset=<1 0 0>
     }]
    relativeSpaceBone {boneName=pelvis}
   }
   n7 = Anim Dynamics - Bone: dyn_strap_l2_03(ComponentPose=@n6, bAlphaBoolEnabled=True, ExternalForce="0,0,0") {
    alphaInputType=Bool
    angularDampingOverride=1
    angularSpringConstant=1000
    bAngularSpring=true
    bOverrideAngularDamping=true
    bUseGravityOverride=true
    bUsePlanarLimit=false
    bUseSphericalLimits=true
    simulationSpace=World
    alphaBoolBlend {blendInTime=0.1, blendOption=Linear, blendOutTime=0.1}
    boundBone {boneName=dyn_strap_l2_03}
    componentAppliedLinearAccClamp=<1000 1000 1000>
    componentLinearAccScale=<0.3 0.3 0.3>
    componentLinearVelScale=<0.3 0.3 0.3>
    gravityOverride {z=-980}
    physicsBodyDefinitions [{
      collisionType=CustomSphere
      sphereCollisionRadius=2.5
      boundBone {boneName=dyn_strap_l2_03}
      boxExtents=<12 8 8>
      constraintSetup {angularLimitsMax=<15 20 -12>, angularLimitsMin=<-15 -20 -25>, angularTarget=<0 0 0>}
      localJointOffset=<1 0 0>
     }]
    relativeSpaceBone {boneName=pelvis}
    sphericalLimits [
     {limitRadius=7, drivingBone {boneName=thigh_twist_01_l}, sphereLocalOffset=<0 0 0>}
     {limitRadius=29, drivingBone {boneName=root}, sphereLocalOffset=<0 0 0>}
    ]
   }
   n8 = Anim Dynamics - Bone: dyn_strap_l2_04(ComponentPose=@n7, bAlphaBoolEnabled=True, ExternalForce="0,0,0") {
    alphaInputType=Bool
    angularDampingOverride=1
    angularSpringConstant=1000
    bAngularSpring=true
    bOverrideAngularDamping=true
    bUseGravityOverride=true
    bUsePlanarLimit=false
    bUseSphericalLimits=true
    simulationSpace=World
    alphaBoolBlend {blendInTime=0.1, blendOption=Linear, blendOutTime=0.1}
    boundBone {boneName=dyn_strap_l2_04}
    componentAppliedLinearAccClamp=<1000 1000 1000>
    componentLinearAccScale=<0.3 0.3 0.3>
    componentLinearVelScale=<0.3 0.3 0.3>
    gravityOverride {z=-980}
    physicsBodyDefinitions [{
      collisionType=CustomSphere
      sphereCollisionRadius=1
      boundBone {boneName=dyn_strap_l2_04}
      boxExtents=<12 8 8>
      constraintSetup {angularLimitsMax=<15 20 10>, angularLimitsMin=<-15 -20 -25>, angularTarget=<0 0 0>}
      localJointOffset=<1 0 0>
     }]
    relativeSpaceBone {boneName=pelvis}
    sphericalLimits [
     {limitRadius=7, drivingBone {boneName=thigh_twist_01_l}, sphereLocalOffset=<0 0 0>}
     {limitRadius=35, drivingBone {boneName=root}, sphereLocalOffset=<0 0 0>}
     {limitRadius=8, drivingBone {boneName=thigh_twist_01_l}, sphereLocalOffset=<7 0 0>}
    ]
   }
   n9 = Anim Dynamics - Bone: dyn_strap_l2_03(ComponentPose=@n8, bAlphaBoolEnabled=True, ExternalForce="0,0,0") {
    alphaInputType=Bool
    angularDampingOverride=1
    angularSpringConstant=1000
    bAngularSpring=true
    bOverrideAngularDamping=true
    bUseGravityOverride=true
    bUsePlanarLimit=false
    bUseSphericalLimits=true
    simulationSpace=World
    alphaBoolBlend {blendInTime=0.1, blendOption=Linear, blendOutTime=0.1}
    boundBone {boneName=dyn_strap_l2_03}
    componentAppliedLinearAccClamp=<1000 1000 1000>
    componentLinearAccScale=<0.3 0.3 0.3>
    componentLinearVelScale=<0.3 0.3 0.3>
    gravityOverride {z=-980}
    physicsBodyDefinitions [{
      collisionType=CustomSphere
      sphereCollisionRadius=2.5
      boundBone {boneName=dyn_strap_l2_03}
      boxExtents=<8 4 4>
      constraintSetup {angularLimitsMax=<15 20 7>, angularLimitsMin=<-15 -20 -35>, angularTarget=<0 0 0>}
      localJointOffset=<1 0 0>
     }]
    relativeSpaceBone {boneName=pelvis}
    sphericalLimits [
     {limitRadius=7, drivingBone {boneName=thigh_twist_01_l}, sphereLocalOffset=<0 0 0>}
     {limitRadius=29, drivingBone {boneName=root}, sphereLocalOffset=<0 0 0>}
    ]
   }
   n10 = Anim Dynamics - Bone: dyn_strap_l2_04(ComponentPose=@n9, bAlphaBoolEnabled=True, ExternalForce="0,0,0") {
    alphaInputType=Bool
    angularDampingOverride=1
    angularSpringConstant=1000
    bAngularSpring=true
    bOverrideAngularDamping=true
    bUseGravityOverride=true
    bUsePlanarLimit=false
    bUseSphericalLimits=true
    simulationSpace=World
    alphaBoolBlend {blendInTime=0.1, blendOption=Linear, blendOutTime=0.1}
    boundBone {boneName=dyn_strap_l2_04}
    componentAppliedLinearAccClamp=<1000 1000 1000>
    componentLinearAccScale=<0.3 0.3 0.3>
    componentLinearVelScale=<0.3 0.3 0.3>
    gravityOverride {z=-980}
    physicsBodyDefinitions [{
      collisionType=CustomSphere
      sphereCollisionRadius=2.5
      boundBone {boneName=dyn_strap_l2_04}
      boxExtents=<8 4 4>
      constraintSetup {angularLimitsMax=<15 20 7>, angularLimitsMin=<-15 -20 -35>, angularTarget=<0 0 0>}
      localJointOffset=<1 0 0>
     }]
    relativeSpaceBone {boneName=pelvis}
    sphericalLimits [
     {limitRadius=7, drivingBone {boneName=thigh_twist_01_l}, sphereLocalOffset=<0 0 0>}
     {limitRadius=35, drivingBone {boneName=root}, sphereLocalOffset=<0 0 0>}
     {limitRadius=9, drivingBone {boneName=thigh_twist_01_l}, sphereLocalOffset=<7 0 0>}
    ]
   }
   n11 = Anim Dynamics - Bone: dyn_strap_l2_03(ComponentPose=@n10, bAlphaBoolEnabled=True, ExternalForce="0,0,0") {
    alphaInputType=Bool
    angularDampingOverride=1
    angularSpringConstant=1000
    bAngularSpring=true
    bOverrideAngularDamping=true
    bUseGravityOverride=true
    bUsePlanarLimit=false
    bUseSphericalLimits=true
    simulationSpace=RootRelative
    alphaBoolBlend {blendInTime=0.1, blendOption=Linear, blendOutTime=0.1}
    boundBone {boneName=dyn_strap_l2_03}
    componentAppliedLinearAccClamp=<1000 1000 1000>
    componentLinearAccScale=<0.3 0.3 0.3>
    componentLinearVelScale=<0.3 0.3 0.3>
    gravityOverride {z=-980}
    physicsBodyDefinitions [{
      collisionType=CustomSphere
      sphereCollisionRadius=2.5
      boundBone {boneName=dyn_strap_l2_03}
      boxExtents=<8 4 8>
      constraintSetup {angularLimitsMax=<15 15 7>, angularLimitsMin=<-15 -15 -25>, angularTarget=<0 0 0>}
      localJointOffset=<1 0 0>
     }]
    relativeSpaceBone {boneName=pelvis}
    sphericalLimits [
     {limitRadius=7, drivingBone {boneName=thigh_twist_01_l}, sphereLocalOffset=<0 0 0>}
     {limitRadius=29, drivingBone {boneName=root}, sphereLocalOffset=<0 0 0>}
    ]
   }
   n12 = Anim Dynamics - Bone: dyn_strap_l2_04(ComponentPose=@n11, bAlphaBoolEnabled=True, ExternalForce="0,0,0") {
    alphaInputType=Bool
    angularDampingOverride=1
    angularSpringConstant=1000
    bAngularSpring=true
    bOverrideAngularDamping=true
    bUseGravityOverride=true
    bUsePlanarLimit=false
    bUseSphericalLimits=true
    simulationSpace=RootRelative
    alphaBoolBlend {blendInTime=0.1, blendOption=Linear, blendOutTime=0.1}
    boundBone {boneName=dyn_strap_l2_04}
    componentAppliedLinearAccClamp=<1000 1000 1000>
    componentLinearAccScale=<0.3 0.3 0.3>
    componentLinearVelScale=<0.3 0.3 0.3>
    gravityOverride {z=-980}
    physicsBodyDefinitions [{
      collisionType=CustomSphere
      sphereCollisionRadius=2.5
      boundBone {boneName=dyn_strap_l2_04}
      boxExtents=<8 4 8>
      constraintSetup {angularLimitsMax=<15 15 7>, angularLimitsMin=<-15 -15 -25>, angularTarget=<0 0 0>}
      localJointOffset=<1 0 0>
     }]
    relativeSpaceBone {boneName=pelvis}
    sphericalLimits [
     {limitRadius=7, drivingBone {boneName=thigh_twist_01_l}, sphereLocalOffset=<0 0 0>}
     {limitRadius=35, drivingBone {boneName=root}, sphereLocalOffset=<0 0 0>}
     {limitRadius=9, drivingBone {boneName=thigh_twist_01_l}, sphereLocalOffset=<7 0 0>}
    ]
   }
   n13 = Anim Dynamics - Bone: dyn_strap_r2_01!(ComponentPose=@n12, bAlphaBoolEnabled:bool=True, ExternalForce="0,0,0") -> (Pose) {
    alphaInputType=Bool
    angularDampingOverride=1
    angularSpringConstant=2000
    bAngularSpring=true
    bOverrideAngularDamping=true
    bUseGravityOverride=true
    simulationSpace=BoneRelative
    alphaBoolBlend {blendInTime=0.1, blendOption=Custom, blendOutTime=0.1}
    boundBone {boneName=dyn_strap_r2_01}
    componentAppliedLinearAccClamp=<1000 1000 1000>
    componentLinearAccScale=<0.3 0.3 0.3>
    componentLinearVelScale=<0.3 0.3 0.3>
    gravityOverride {z=-980}
    physicsBodyDefinitions [{
      boundBone {boneName=dyn_strap_r2_01}
      boxExtents=<15 5 15>
      constraintSetup {angularLimitsMax=<0 20 25>, angularLimitsMin=<0 -20 0>, angularTarget=<0 0 0>}
      localJointOffset=<1 0 0>
     }]
    relativeSpaceBone {boneName=root}
    sphericalLimits [{limitRadius=8, drivingBone {boneName=thigh_r}, sphereLocalOffset=<0 0 0>}]
   }
   n14 = Anim Dynamics - Bone: dyn_strap_r2_02!(ComponentPose=@n13, bAlphaBoolEnabled:bool=True, ExternalForce="0,0,0") -> (Pose) {
    alphaInputType=Bool
    angularDampingOverride=1
    angularSpringConstant=1000
    bAngularSpring=true
    bOverrideAngularDamping=true
    bUseGravityOverride=true
    bUsePlanarLimit=false
    bUseSphericalLimits=true
    simulationSpace=RootRelative
    alphaBoolBlend {blendInTime=0.1, blendOption=Linear, blendOutTime=0.1}
    boundBone {boneName=dyn_strap_r2_02}
    componentAppliedLinearAccClamp=<1000 1000 1000>
    componentLinearAccScale=<0.3 0.3 0.3>
    componentLinearVelScale=<0.3 0.3 0.3>
    gravityOverride {z=-980}
    physicsBodyDefinitions [{
      collisionType=CustomSphere
      sphereCollisionRadius=2.5
      boundBone {boneName=dyn_strap_r2_02}
      boxExtents=<8 4 8>
      constraintSetup {angularLimitsMax=<5 0 5>, angularLimitsMin=<-5 0 0>, angularTarget=<0 0 0>}
      localJointOffset=<1 0 0>
     }]
    relativeSpaceBone {boneName=pelvis}
   }
   n15 = Anim Dynamics - Bone: dyn_strap_r2_01(ComponentPose=@n14, bAlphaBoolEnabled=True, ExternalForce="0,0,0") {
    alphaInputType=Bool
    angularDampingOverride=1
    angularSpringConstant=2000
    bAngularSpring=true
    bOverrideAngularDamping=true
    bUseGravityOverride=true
    simulationSpace=BoneRelative
    alphaBoolBlend {blendInTime=0.1, blendOption=Custom, blendOutTime=0.1}
    boundBone {boneName=dyn_strap_r2_01}
    componentAppliedLinearAccClamp=<1000 1000 1000>
    componentLinearAccScale=<0.3 0.3 0.3>
    componentLinearVelScale=<0.3 0.3 0.3>
    gravityOverride {z=-980}
    physicsBodyDefinitions [{
      boundBone {boneName=dyn_strap_r2_01}
      boxExtents=<15 5 15>
      constraintSetup {angularLimitsMax=<0 20 0>, angularLimitsMin=<0 -20 0>, angularTarget=<0 0 0>}
      localJointOffset=<1 0 0>
     }]
    relativeSpaceBone {boneName=root}
    sphericalLimits [{limitRadius=8, drivingBone {boneName=thigh_r}, sphereLocalOffset=<0 0 0>}]
   }
   n16 = Anim Dynamics - Bone: dyn_strap_r2_02(ComponentPose=@n15, bAlphaBoolEnabled=True, ExternalForce="0,0,0") {
    alphaInputType=Bool
    angularDampingOverride=1
    angularSpringConstant=1000
    bAngularSpring=true
    bOverrideAngularDamping=true
    bUseGravityOverride=true
    bUsePlanarLimit=false
    bUseSphericalLimits=true
    simulationSpace=RootRelative
    alphaBoolBlend {blendInTime=0.1, blendOption=Linear, blendOutTime=0.1}
    boundBone {boneName=dyn_strap_r2_02}
    componentAppliedLinearAccClamp=<1000 1000 1000>
    componentLinearAccScale=<0.3 0.3 0.3>
    componentLinearVelScale=<0.3 0.3 0.3>
    gravityOverride {z=-980}
    physicsBodyDefinitions [{
      collisionType=CustomSphere
      sphereCollisionRadius=2.5
      boundBone {boneName=dyn_strap_r2_02}
      boxExtents=<8 4 8>
      constraintSetup {angularLimitsMax=<5 0 0>, angularLimitsMin=<-5 0 0>, angularTarget=<0 0 0>}
      localJointOffset=<1 0 0>
     }]
    relativeSpaceBone {boneName=pelvis}
   }
   n17 = Anim Dynamics - Bone: dyn_strap_r1_01!(ComponentPose=@n16, bAlphaBoolEnabled:bool=True, ExternalForce="0,0,0") -> (Pose) {
    alphaInputType=Bool
    angularDampingOverride=1
    angularSpringConstant=2000
    bAngularSpring=true
    bOverrideAngularDamping=true
    bUseGravityOverride=true
    bUseSphericalLimits=true
    simulationSpace=BoneRelative
    alphaBoolBlend {blendInTime=0.1, blendOption=Custom, blendOutTime=0.1}
    boundBone {boneName=dyn_strap_r1_01}
    componentAppliedLinearAccClamp=<1000 1000 1000>
    componentLinearAccScale=<0.3 0.3 0.3>
    componentLinearVelScale=<0.3 0.3 0.3>
    gravityOverride {z=-980}
    physicsBodyDefinitions [{
      boundBone {boneName=dyn_strap_r1_01}
      boxExtents=<10 5 10>
      constraintSetup {angularLimitsMax=<0 10 15>, angularLimitsMin=<0 -10 0>, angularTarget=<0 0 0>}
      localJointOffset=<1 0 0>
     }]
    relativeSpaceBone {boneName=root}
    sphericalLimits [{limitRadius=8, drivingBone {boneName=thigh_r}, sphereLocalOffset=<0 0 0>}]
   }
   n18 = Anim Dynamics - Bone: dyn_strap_r3_01!(ComponentPose=@n17, bAlphaBoolEnabled:bool=True, ExternalForce="0,0,0") -> (Pose) {
    alphaInputType=Bool
    angularDampingOverride=1
    angularSpringConstant=2000
    bAngularSpring=true
    bOverrideAngularDamping=true
    bUseGravityOverride=true
    bUseSphericalLimits=true
    simulationSpace=BoneRelative
    alphaBoolBlend {blendInTime=0.1, blendOption=Custom, blendOutTime=0.1}
    boundBone {boneName=dyn_strap_r3_01}
    componentAppliedLinearAccClamp=<1000 1000 1000>
    componentLinearAccScale=<0.3 0.3 0.3>
    componentLinearVelScale=<0.3 0.3 0.3>
    gravityOverride {z=-980}
    physicsBodyDefinitions [{
      boundBone {boneName=dyn_strap_r3_01}
      boxExtents=<10 5 10>
      constraintSetup {angularLimitsMax=<0 0 15>, angularLimitsMin=<0 -10 0>, angularTarget=<0 0 0>}
      localJointOffset=<1 0 0>
     }]
    relativeSpaceBone {boneName=root}
    sphericalLimits [{limitRadius=8, drivingBone {boneName=thigh_r}, sphereLocalOffset=<0 0 0>}]
   }
   n19 = Anim Dynamics - Bone: dyn_strap_r3_02!(ComponentPose=@n18, bAlphaBoolEnabled:bool=True, ExternalForce="0,0,0") -> (Pose) {
    alphaInputType=Bool
    angularDampingOverride=1
    angularSpringConstant=2000
    bAngularSpring=true
    bOverrideAngularDamping=true
    bUseGravityOverride=true
    bUseSphericalLimits=true
    simulationSpace=BoneRelative
    alphaBoolBlend {blendInTime=0.1, blendOption=Custom, blendOutTime=0.1}
    boundBone {boneName=dyn_strap_r3_02}
    componentAppliedLinearAccClamp=<1000 1000 1000>
    componentLinearAccScale=<0.3 0.3 0.3>
    componentLinearVelScale=<0.3 0.3 0.3>
    gravityOverride {z=-980}
    physicsBodyDefinitions [{
      boundBone {boneName=dyn_strap_r3_02}
      boxExtents=<10 5 10>
      constraintSetup {angularLimitsMax=<0 0 15>, angularLimitsMin=<0 -10 0>, angularTarget=<0 0 0>}
      localJointOffset=<1 0 0>
     }]
    relativeSpaceBone {boneName=root}
    sphericalLimits [{limitRadius=8, drivingBone {boneName=thigh_r}, sphereLocalOffset=<0 0 0>}]
   }
   Local To Component(LocalPose=Blend Poses by bool(BlendPose_0=Component To Local(), BlendPose_1=Component To Local(ComponentPose=@n19), bActiveValue=Get Zipline() -> (Zipline:bool)) {blendType=Linear})
   Local To Component(LocalPose=Use cached pose 'Presim'())
   Look At - Bone: dyn_strap_r1_01_MSTR!(ComponentPose, bAlphaBoolEnabled:bool=True) -> (Pose) {
    alphaInputType=Bool
    alphaBoolBlend {blendOption=Linear}
    boneToModify {boneName=dyn_strap_r1_01_MSTR}
    lookAtLocation=<43.7889 -16.0367 -0.8944>
    lookAtTarget {boneReference {boneName=thigh_r}}
    lookAt_Axis {axis=<10 0>}
   }
   n20 = Look At - Bone: dyn_strap_l2_01_MSTR(Alpha=1) {
    alphaBoolBlend {blendOption=Linear}
    boneToModify {boneName=dyn_strap_l2_01_MSTR}
    lookAtLocation=<-41.771 7.601 -17.0179>
    lookAtTarget {boneReference {boneName=thigh_l}}
    lookAt_Axis {axis=<1 0>}
   }
   n21 = Look At - Bone: dyn_strap_l1_01_MSTR(ComponentPose=@n20, Alpha=1) {
    alphaBoolBlend {blendOption=Linear}
    boneToModify {boneName=dyn_strap_l1_01_MSTR}
    lookAtLocation=<-41.771 7.601 -17.0179>
    lookAtTarget {boneReference {boneName=thigh_l}}
    lookAt_Axis {axis=<1 0>}
   }
   n22 = Look At - Bone: dyn_strap_r2_01_MSTR(ComponentPose=@n21, Alpha=1) {
    alphaBoolBlend {blendOption=Linear}
    boneToModify {boneName=dyn_strap_r2_01_MSTR}
    lookAtLocation=<46.6039 -7.1125 19.1147>
    lookAtTarget {boneReference {boneName=thigh_r}}
    lookAt_Axis {axis=<1 0>}
   }
   Look At - Bone: dyn_strap_r1_01_MSTR!(ComponentPose=@n22, Alpha:float=1) -> (Pose) {
    alphaBoolBlend {blendOption=Linear}
    boneToModify {boneName=dyn_strap_r1_01_MSTR}
    lookAtLocation=<43.7889 -16.0367 -0.8944>
    lookAtTarget {boneReference {boneName=thigh_r}}
    lookAt_Axis {axis=<10 0>}
   }
   Look At - Bone: dyn_strap_r2_01_MSTR(Alpha=1) {
    alphaBoolBlend {blendOption=Linear}
    boneToModify {boneName=dyn_strap_r2_01_MSTR}
    lookAtLocation=<24.6249 -7.1125 15>
    lookAtTarget {boneReference {boneName=thigh_r}}
    lookAt_Axis {axis=<1 0>}
   }
   n23 = Look At - Bone: dyn_strap_r2_01_MSTR(Alpha=1) {
    alphaBoolBlend {blendOption=Linear}
    boneToModify {boneName=dyn_strap_r2_01_MSTR}
    lookAtLocation=<46.6039 -7.1125 15>
    lookAtTarget {boneReference {boneName=thigh_r}}
    lookAt_Axis {axis=<1 0>}
   }
   Look At - Bone: dyn_strap_r1_01_MSTR(ComponentPose=@n23, Alpha=1) {
    alphaBoolBlend {blendOption=Linear}
    boneToModify {boneName=dyn_strap_r1_01_MSTR}
    lookAtLocation=<43.7889 -16.0367 -0.8944>
    lookAtTarget {boneReference {boneName=thigh_r}}
    lookAt_Axis {axis=<10 0>}
   }
   "Transform (Modify) Bone"(Alpha=1, Translation="0,0,0", Rotation="0,0,0", Scale="1,1,1") {alphaBoolBlend {blendOption=Linear}, boneToModify=~}
   n24 = "Transform (Modify) Bone - Bone: dyn_spurs_l"(
    ComponentPose=Local To Component(LocalPose=Use cached pose 'CRCache_Dynamics'())
    Alpha=1
    Translation="0,0,0"
    Rotation="Get Left Spur Rotation "() -> ("Left Spur Rotation ":Rotator)
    Scale="1,1,1"
   ) {rotationMode=BMM_Additive, rotationSpace=BCS_BoneSpace, alphaBoolBlend {blendOption=Linear}, boneToModify {boneName=dyn_spurs_l}}
   n25 = "Transform (Modify) Bone - Bone: dyn_spurs_r"(ComponentPose=@n24, Alpha=1, Translation="0,0,0", Rotation=Get Spin result() -> (Spin result:Rotator), Scale="1,1,1") {rotationMode=BMM_Additive, rotationSpace=BCS_BoneSpace, alphaBoolBlend {blendOption=Linear}, boneToModify {boneName=dyn_spurs_r}}
   n26 = "Transform (Modify) Bone - Bone: dyn_pistol_spurs_r"(ComponentPose=@n25, Alpha=1, Translation="0,0,0", Rotation=Get Pistol Spur Rotation() -> (Pistol Spur Rotation:Rotator), Scale="1,1,1") {rotationMode=BMM_Additive, rotationSpace=BCS_BoneSpace, alphaBoolBlend {blendOption=Linear}, boneToModify {boneName=dyn_pistol_spurs_r}}
   Output Pose(Result=Component To Local(ComponentPose=@n26)) {layerGroup=None}
   n27 = "Constraint - dyn_spurs_l (PP)"!(
    ComponentPose=Local To Component(LocalPose=Copy Pose From Mesh() {bUseAttachedParent=true})
    Alpha=1
    ConstraintWeights_0=0.1202
    ConstraintWeights_1:float=0.8798
   ) -> (Pose) {
    alphaBoolBlend {blendOption=Linear}
    boneToModify {boneName=dyn_spurs_l}
    constraintSetup [{targetBone {boneName=calf_twist_01_l}}, {targetBone {boneName=foot_l}}]
    constraintWeights [1,1]
   }
   n28 = "Constraint - dyn_spurs_r (PP)"!(ComponentPose=@n27, Alpha=1, ConstraintWeights_0=0.1199, ConstraintWeights_1:float=0.8801) -> (Pose) {
    alphaBoolBlend {blendOption=Linear}
    boneToModify {boneName=dyn_spurs_r}
    constraintSetup [{targetBone {boneName=calf_twist_01_r}}, {targetBone {boneName=foot_r}}]
    constraintWeights [1,1]
   }
   n29 = "Constraint - dyn_buckle_01 (PPPPPPPP)"!(
    ComponentPose=@n28
    Alpha=1
    ConstraintWeights_0=0.1119
    ConstraintWeights_1:float=0.0109
    ConstraintWeights_2:float=0.766
    ConstraintWeights_3:float=0.0309
    ConstraintWeights_4:float=0.0324
    ConstraintWeights_5:float=0.0057
    ConstraintWeights_6:float=0.0239
    ConstraintWeights_7:float=0.0182
   ) -> (Pose) {
    alphaBoolBlend {blendOption=Linear}
    boneToModify {boneName=dyn_buckle_01}
    constraintSetup [
     {targetBone {boneName=spine_02}}
     {targetBone {boneName=spine_04}}
     {targetBone {boneName=pelvis}}
     {targetBone {boneName=thigh_twist_01_l}}
     {targetBone {boneName=thigh_twist_01_r}}
     {targetBone {boneName=thigh_r}}
     {targetBone {boneName=deform_groin_r}}
     {targetBone {boneName=deform_groin_l}}
    ]
    constraintWeights [1,1,1,1,1,1,1,1]
   }
   n30 = Control Rig!(Source=Component To Local(ComponentPose=@n29), Alpha, getCrouch:bool=Get bIsCrouching()) -> (Pose) {
    alphaBoolBlend {blendOption=Linear}
    controlRigAssetReference {blueprintRigClass=&/F_MED_NeonGlow_CtrlRig.F_MED_NeonGlow_CtrlRig_C}
   }
   Save cached pose 'CRCache_InputPose'(Pose=@n30) {cacheName=CRCache_InputPose}
   n31 = Look At - Bone: dyn_strap_l2_01_MSTR!(ComponentPose, bAlphaBoolEnabled:bool=True) -> (Pose) {
    alphaInputType=Bool
    alphaBoolBlend {blendOption=Linear}
    boneToModify {boneName=dyn_strap_l2_01_MSTR}
    lookAtLocation {x=-41.771, z=-15}
    lookAtTarget {boneReference {boneName=thigh_l}}
    lookAt_Axis {axis=<1 0>}
   }
   n32 = Look At - Bone: dyn_strap_l1_01_MSTR!(ComponentPose=@n31, bAlphaBoolEnabled:bool=True) -> (Pose) {
    alphaInputType=Bool
    alphaBoolBlend {blendOption=Linear}
    boneToModify {boneName=dyn_strap_l1_01_MSTR}
    lookAtLocation=<-41.771 7.601 -17.0179>
    lookAtTarget {boneReference {boneName=thigh_l}}
    lookAt_Axis {axis=<1 0>}
   }
   n33 = Look At - Bone: dyn_strap_r2_01_MSTR!(ComponentPose=@n32, bAlphaBoolEnabled:bool=True) -> (Pose) {
    alphaInputType=Bool
    alphaBoolBlend {blendOption=Linear}
    boneToModify {boneName=dyn_strap_r2_01_MSTR}
    lookAtLocation=<46.6039 -7.1125 11>
    lookAtTarget {boneReference {boneName=thigh_r}}
    lookAt_Axis {axis=<1 0>}
   }
   n34 = Look At - Bone: dyn_strap_r1_01_MSTR!(ComponentPose=@n33, bAlphaBoolEnabled:bool=True) -> (Pose) {
    alphaInputType=Bool
    alphaBoolBlend {blendOption=Linear}
    boneToModify {boneName=dyn_strap_r1_01_MSTR}
    lookAtLocation=<43.7889 -16.0367 -0.8944>
    lookAtTarget {boneReference {boneName=thigh_r}}
    lookAt_Axis {axis=<10 0>}
   }
   n35 = Look At - Bone: dyn_strap_l2_01_MSTR(ComponentPose=@n34, bAlphaBoolEnabled=True) {
    alphaInputType=Bool
    alphaBoolBlend {blendOption=Linear}
    boneToModify {boneName=dyn_strap_l2_01_MSTR}
    lookAtLocation=<-41.771 7.601 -17.0179>
    lookAtTarget {boneReference {boneName=thigh_l}}
    lookAt_Axis {axis=<1 0>}
   }
   n36 = Look At - Bone: dyn_strap_l1_01_MSTR(ComponentPose=@n35, bAlphaBoolEnabled=True) {
    alphaInputType=Bool
    alphaBoolBlend {blendOption=Linear}
    boneToModify {boneName=dyn_strap_l1_01_MSTR}
    lookAtLocation=<-41.771 7.601 -17.0179>
    lookAtTarget {boneReference {boneName=thigh_l}}
    lookAt_Axis {axis=<1 0>}
   }
   n37 = Look At - Bone: dyn_strap_r2_01_MSTR(ComponentPose=@n36, bAlphaBoolEnabled=True) {
    alphaInputType=Bool
    alphaBoolBlend {blendOption=Linear}
    boneToModify {boneName=dyn_strap_r2_01_MSTR}
    lookAtLocation=<46.6039 -7.1125 15>
    lookAtTarget {boneReference {boneName=thigh_r}}
    lookAt_Axis {axis=<1 0>}
   }
   n38 = Look At - Bone: dyn_strap_l1_01_MSTR(ComponentPose=@n37, bAlphaBoolEnabled=True) {
    alphaInputType=Bool
    alphaBoolBlend {blendOption=Linear}
    boneToModify {boneName=dyn_strap_l1_01_MSTR}
    lookAtLocation=<-41.771 7.601 -17.0179>
    lookAtTarget {boneReference {boneName=thigh_l}}
    lookAt_Axis {axis=<1 0>}
   }
   n39 = Look At - Bone: dyn_strap_l2_01_MSTR(ComponentPose=@n38, bAlphaBoolEnabled=True) {
    alphaInputType=Bool
    alphaBoolBlend {blendOption=Linear}
    boneToModify {boneName=dyn_strap_l2_01_MSTR}
    lookAtLocation=<-41.7835 7.5181 -16.2673>
    lookAtTarget {boneReference {boneName=thigh_l}}
    lookAt_Axis {axis=<1 0>}
   }
   n40 = Look At - Bone: dyn_strap_l1_01_MSTR(ComponentPose=@n39, bAlphaBoolEnabled=True) {
    alphaInputType=Bool
    alphaBoolBlend {blendOption=Linear}
    boneToModify {boneName=dyn_strap_l1_01_MSTR}
    lookAtLocation=<-41.771 7.601 -17.0179>
    lookAtTarget {boneReference {boneName=thigh_l}}
    lookAt_Axis {axis=<1 0>}
   }
   n41 = Look At - Bone: dyn_strap_r2_01_MSTR(ComponentPose=@n40, bAlphaBoolEnabled=True) {
    alphaInputType=Bool
    alphaBoolBlend {blendOption=Linear}
    boneToModify {boneName=dyn_strap_r2_01_MSTR}
    lookAtLocation=<46.6039 -7.1125 15>
    lookAtTarget {boneReference {boneName=thigh_r}}
    lookAt_Axis {axis=<1 0>}
   }
   n42 = Look At - Bone: dyn_strap_l1_01_MSTR(ComponentPose=@n41, bAlphaBoolEnabled=True) {
    alphaInputType=Bool
    alphaBoolBlend {blendOption=Linear}
    boneToModify {boneName=dyn_strap_l1_01_MSTR}
    lookAtLocation=<-41.771 7.601 -17.0179>
    lookAtTarget {boneReference {boneName=thigh_l}}
    lookAt_Axis {axis=<1 0>}
   }
   Save cached pose 'Presim'(Pose=Component To Local(ComponentPose=@n42)) {cacheName=Presim}
   n43 = "Transform (Modify) Bone - Bone: dyn_strap_l3_01_MSTR"!(
    ComponentPose=Local To Component(LocalPose=Use cached pose 'CRCache_InputPose'())
    bAlphaBoolEnabled:bool=True
    Translation="0,0,0"
    Rotation="0,-6,0"
    Scale="1,1,1"
   ) -> (Pose) {
    alphaInputType=Bool
    rotationMode=BMM_Additive
    rotationSpace=BCS_BoneSpace
    alphaBoolBlend {blendInTime=0.2, blendOption=Linear, blendOutTime=0.2}
    boneToModify {boneName=dyn_strap_l3_01_MSTR}
    rotation {yaw=-6}
   }
   n44 = RigidBody(ComponentPose=@n43) {
    bOverrideWorldGravity=true
    overridePhysicsAsset=&/F_MED_NeonGlow_Physics
    simulationSpace=BaseBoneSpace
    alphaBoolBlend {blendOption=Linear}
    baseBoneRef {boneName=pelvis}
    overrideWorldGravity {z=-980}
    simSpaceSettings {
     maxAngularAcceleration=3
     maxAngularVelocity=1
     maxLinearAcceleration=900
     maxLinearVelocity=300
     worldAlpha=1
     externalLinearDragV=<0.2 0.2 0.2>
    }
   }
   n45 = Anim Dynamics - Bone: dyn_strap_l2_01!(ComponentPose=@n44, Alpha:float=1, ExternalForce="0,0,0") -> (Pose) {
    angularDampingOverride=1
    angularSpringConstant=500
    bAngularSpring=true
    bOverrideAngularDamping=true
    bUseGravityOverride=true
    gravityScale=0.35
    alphaBoolBlend {blendOption=Linear}
    boundBone {boneName=dyn_strap_l2_01}
    componentAppliedLinearAccClamp=<3000 3000 3000>
    componentLinearAccScale=<0.3 0.3 0.3>
    componentLinearVelScale=<0.3 0.3 0.3>
    gravityOverride {z=-980}
    physicsBodyDefinitions [{
      boundBone {boneName=dyn_strap_l2_01}
      boxExtents=<30 30 30>
      constraintSetup {angularLimitsMax=<0 0 -1>, angularLimitsMin=<0 0 -15>, angularTarget=<1 0 0>}
      localJointOffset=<2 0 0>
     }]
    relativeSpaceBone {boneName=thigh_l}
   }
   n46 = "Transform (Modify) Bone - Bone: dyn_strap_l3_02"!(ComponentPose=@n45, bAlphaBoolEnabled:bool=True, Translation="-0.000,1.284,-0.000", Rotation="0.000,2.240,-0.000", Scale="1,1,1") -> (Pose) {
    alphaInputType=Bool
    rotationMode=BMM_Additive
    rotationSpace=BCS_BoneSpace
    translationMode=BMM_Additive
    translationSpace=BCS_BoneSpace
    alphaBoolBlend {blendInTime=0.2, blendOption=Linear, blendOutTime=0.2}
    boneToModify {boneName=dyn_strap_l3_02}
    rotation=<rot 0.0001 2.24 -0.0001>
    translation=<0 1.2842 0>
   }
   Save cached pose 'CRCache_Dynamics'(Pose=Component To Local(ComponentPose=@n46)) {cacheName=CRCache_Dynamics}
   n47 = Anim Dynamics - Bone: dyn_strap_l2_01!(ComponentPose, bAlphaBoolEnabled:bool=True, ExternalForce="0,0,0") -> (Pose) {
    alphaInputType=Bool
    angularDampingOverride=1
    angularSpringConstant=2000
    bAngularSpring=true
    bOverrideAngularDamping=true
    bUseGravityOverride=true
    bUseSphericalLimits=true
    simulationSpace=BoneRelative
    alphaBoolBlend {blendInTime=0.1, blendOption=Custom, blendOutTime=0.1}
    boundBone {boneName=dyn_strap_l2_01}
    componentAppliedLinearAccClamp=<1000 1000 1000>
    componentLinearAccScale=<0.3 0.3 0.3>
    componentLinearVelScale=<0.3 0.3 0.3>
    gravityOverride {z=-980}
    physicsBodyDefinitions [{
      boundBone {boneName=dyn_strap_l2_01}
      boxExtents=<20 10 20>
      constraintSetup {angularLimitsMax=<0 20 0>, angularLimitsMin=<0 -20 -25>, angularTarget=<0 0 0>}
      localJointOffset=<1 0 0>
     }]
    relativeSpaceBone {boneName=root}
    sphericalLimits [{limitRadius=8, drivingBone {boneName=thigh_l}, sphereLocalOffset=<0 0 0>}]
   }
   n48 = Anim Dynamics - Bone: dyn_strap_l1_01(ComponentPose=@n47, bAlphaBoolEnabled=True, ExternalForce="0,0,0") {
    alphaInputType=Bool
    angularDampingOverride=1
    angularSpringConstant=2000
    bAngularSpring=true
    bOverrideAngularDamping=true
    bUseGravityOverride=true
    bUseSphericalLimits=true
    simulationSpace=BoneRelative
    alphaBoolBlend {blendInTime=0.1, blendOption=Custom, blendOutTime=0.1}
    boundBone {boneName=dyn_strap_l1_01}
    componentAppliedLinearAccClamp=<1000 1000 1000>
    componentLinearAccScale=<0.3 0.3 0.3>
    componentLinearVelScale=<0.3 0.3 0.3>
    gravityOverride {z=-980}
    physicsBodyDefinitions [{
      boundBone {boneName=dyn_strap_l1_01}
      boxExtents=<10 5 10>
      constraintSetup {angularLimitsMax=<0 10 0>, angularLimitsMin=<0 -10 -15>, angularTarget=<0 0 0>}
      localJointOffset=<1 0 0>
     }]
    relativeSpaceBone {boneName=root}
    sphericalLimits [{limitRadius=8, drivingBone {boneName=thigh_l}, sphereLocalOffset=<0 0 0>}]
   }
   n49 = Anim Dynamics - Bone: dyn_strap_l3_01(ComponentPose=@n48, bAlphaBoolEnabled=True, ExternalForce="0,0,0") {
    alphaInputType=Bool
    angularDampingOverride=1
    angularSpringConstant=2000
    bAngularSpring=true
    bOverrideAngularDamping=true
    bUseGravityOverride=true
    bUseSphericalLimits=true
    simulationSpace=BoneRelative
    alphaBoolBlend {blendInTime=0.1, blendOption=Custom, blendOutTime=0.1}
    boundBone {boneName=dyn_strap_l3_01}
    componentAppliedLinearAccClamp=<1000 1000 1000>
    componentLinearAccScale=<0.3 0.3 0.3>
    componentLinearVelScale=<0.3 0.3 0.3>
    gravityOverride {z=-980}
    physicsBodyDefinitions [{
      boundBone {boneName=dyn_strap_l3_01}
      boxExtents=<10 5 10>
      constraintSetup {angularLimitsMax=<0 0 0>, angularLimitsMin=<0 -10 -15>, angularTarget=<0 0 0>}
      localJointOffset=<1 0 0>
     }]
    relativeSpaceBone {boneName=root}
    sphericalLimits [{limitRadius=8, drivingBone {boneName=thigh_l}, sphereLocalOffset=<0 0 0>}]
   }
   n50 = Anim Dynamics - Bone: dyn_strap_l3_02(ComponentPose=@n49, bAlphaBoolEnabled=True, ExternalForce="0,0,0") {
    alphaInputType=Bool
    angularDampingOverride=1
    angularSpringConstant=2000
    bAngularSpring=true
    bOverrideAngularDamping=true
    bUseGravityOverride=true
    bUseSphericalLimits=true
    simulationSpace=BoneRelative
    alphaBoolBlend {blendInTime=0.1, blendOption=Custom, blendOutTime=0.1}
    boundBone {boneName=dyn_strap_l3_02}
    componentAppliedLinearAccClamp=<1000 1000 1000>
    componentLinearAccScale=<0.3 0.3 0.3>
    componentLinearVelScale=<0.3 0.3 0.3>
    gravityOverride {z=-980}
    physicsBodyDefinitions [{
      boundBone {boneName=dyn_strap_l3_02}
      boxExtents=<10 5 10>
      constraintSetup {angularLimitsMax=<0 0 0>, angularLimitsMin=<0 -10 -15>, angularTarget=<0 0 0>}
      localJointOffset=<1 0 0>
     }]
    relativeSpaceBone {boneName=root}
    sphericalLimits [{limitRadius=8, drivingBone {boneName=thigh_l}, sphereLocalOffset=<0 0 0>}]
   }
   n51 = Anim Dynamics - Bone: dyn_strap_l2_02(ComponentPose=@n50, bAlphaBoolEnabled=True, ExternalForce="0,0,0") {
    alphaInputType=Bool
    angularDampingOverride=1
    angularSpringConstant=1000
    bAngularSpring=true
    bOverrideAngularDamping=true
    bUseGravityOverride=true
    bUsePlanarLimit=false
    simulationSpace=RootRelative
    alphaBoolBlend {blendInTime=0.1, blendOption=Linear, blendOutTime=0.1}
    boundBone {boneName=dyn_strap_l2_02}
    componentAppliedLinearAccClamp=<1000 1000 1000>
    componentLinearAccScale=<0.3 0.3 0.3>
    componentLinearVelScale=<0.3 0.3 0.3>
    gravityOverride {z=-980}
    physicsBodyDefinitions [{
      collisionType=CustomSphere
      sphereCollisionRadius=2.5
      boundBone {boneName=dyn_strap_l2_02}
      boxExtents=<8 4 8>
      constraintSetup {angularLimitsMax=<15 15 0>, angularLimitsMin=<-15 -15 -25>, angularTarget=<0 0 0>}
      localJointOffset=<1 0 0>
     }]
    relativeSpaceBone {boneName=pelvis}
   }
   n52 = Anim Dynamics - Bone: dyn_strap_l2_03(ComponentPose=@n51, bAlphaBoolEnabled=True, ExternalForce="0,0,0") {
    alphaInputType=Bool
    angularDampingOverride=1
    angularSpringConstant=1000
    bAngularSpring=true
    bOverrideAngularDamping=true
    bUseGravityOverride=true
    bUsePlanarLimit=false
    bUseSphericalLimits=true
    simulationSpace=RootRelative
    alphaBoolBlend {blendInTime=0.1, blendOption=Linear, blendOutTime=0.1}
    boundBone {boneName=dyn_strap_l2_03}
    componentAppliedLinearAccClamp=<1000 1000 1000>
    componentLinearAccScale=<0.3 0.3 0.3>
    componentLinearVelScale=<0.3 0.3 0.3>
    gravityOverride {z=-980}
    physicsBodyDefinitions [{
      collisionType=CustomSphere
      sphereCollisionRadius=2.5
      boundBone {boneName=dyn_strap_l2_03}
      boxExtents=<8 4 8>
      constraintSetup {angularLimitsMax=<15 15 7>, angularLimitsMin=<-15 -15 -25>, angularTarget=<0 0 0>}
      localJointOffset=<1 0 0>
     }]
    relativeSpaceBone {boneName=pelvis}
    sphericalLimits [
     {limitRadius=7, drivingBone {boneName=thigh_twist_01_l}, sphereLocalOffset=<0 0 0>}
     {limitRadius=29, drivingBone {boneName=root}, sphereLocalOffset=<0 0 0>}
    ]
   }
   n53 = Anim Dynamics - Bone: dyn_strap_l2_04(ComponentPose=@n52, bAlphaBoolEnabled=True, ExternalForce="0,0,0") {
    alphaInputType=Bool
    angularDampingOverride=1
    angularSpringConstant=1000
    bAngularSpring=true
    bOverrideAngularDamping=true
    bUseGravityOverride=true
    bUsePlanarLimit=false
    bUseSphericalLimits=true
    simulationSpace=RootRelative
    alphaBoolBlend {blendInTime=0.1, blendOption=Linear, blendOutTime=0.1}
    boundBone {boneName=dyn_strap_l2_04}
    componentAppliedLinearAccClamp=<1000 1000 1000>
    componentLinearAccScale=<0.3 0.3 0.3>
    componentLinearVelScale=<0.3 0.3 0.3>
    gravityOverride {z=-980}
    physicsBodyDefinitions [{
      collisionType=CustomSphere
      sphereCollisionRadius=2.5
      boundBone {boneName=dyn_strap_l2_04}
      boxExtents=<8 4 8>
      constraintSetup {angularLimitsMax=<15 15 7>, angularLimitsMin=<-15 -15 -25>, angularTarget=<0 0 0>}
      localJointOffset=<1 0 0>
     }]
    relativeSpaceBone {boneName=pelvis}
    sphericalLimits [
     {limitRadius=7, drivingBone {boneName=thigh_twist_01_l}, sphereLocalOffset=<0 0 0>}
     {limitRadius=35, drivingBone {boneName=root}, sphereLocalOffset=<0 0 0>}
     {limitRadius=9, drivingBone {boneName=thigh_twist_01_l}, sphereLocalOffset=<7 0 0>}
    ]
   }
   n54 = Anim Dynamics - Bone: dyn_strap_r2_01(ComponentPose=@n53, bAlphaBoolEnabled=True, ExternalForce="0,0,0") {
    alphaInputType=Bool
    angularDampingOverride=1
    angularSpringConstant=2000
    bAngularSpring=true
    bOverrideAngularDamping=true
    bUseGravityOverride=true
    bUseSphericalLimits=true
    simulationSpace=BoneRelative
    alphaBoolBlend {blendInTime=0.1, blendOption=Custom, blendOutTime=0.1}
    boundBone {boneName=dyn_strap_r2_01}
    componentAppliedLinearAccClamp=<1000 1000 1000>
    componentLinearAccScale=<0.3 0.3 0.3>
    componentLinearVelScale=<0.3 0.3 0.3>
    gravityOverride {z=-980}
    physicsBodyDefinitions [{
      boundBone {boneName=dyn_strap_r2_01}
      boxExtents=<15 5 15>
      constraintSetup {angularLimitsMax=<0 20 25>, angularLimitsMin=<0 -20 0>, angularTarget=<0 0 0>}
      localJointOffset=<1 0 0>
     }]
    relativeSpaceBone {boneName=root}
    sphericalLimits [{limitRadius=8, drivingBone {boneName=thigh_r}, sphereLocalOffset=<0 0 0>}]
   }
   n55 = Anim Dynamics - Bone: dyn_strap_r1_01(ComponentPose=@n54, bAlphaBoolEnabled=True, ExternalForce="0,0,0") {
    alphaInputType=Bool
    angularDampingOverride=1
    angularSpringConstant=2000
    bAngularSpring=true
    bOverrideAngularDamping=true
    bUseGravityOverride=true
    bUseSphericalLimits=true
    simulationSpace=BoneRelative
    alphaBoolBlend {blendInTime=0.1, blendOption=Custom, blendOutTime=0.1}
    boundBone {boneName=dyn_strap_r1_01}
    componentAppliedLinearAccClamp=<1000 1000 1000>
    componentLinearAccScale=<0.3 0.3 0.3>
    componentLinearVelScale=<0.3 0.3 0.3>
    gravityOverride {z=-980}
    physicsBodyDefinitions [{
      boundBone {boneName=dyn_strap_r1_01}
      boxExtents=<10 5 10>
      constraintSetup {angularLimitsMax=<0 10 15>, angularLimitsMin=<0 -10 0>, angularTarget=<0 0 0>}
      localJointOffset=<1 0 0>
     }]
    relativeSpaceBone {boneName=root}
    sphericalLimits [{limitRadius=8, drivingBone {boneName=thigh_r}, sphereLocalOffset=<0 0 0>}]
   }
   n56 = Anim Dynamics - Bone: dyn_strap_r3_01(ComponentPose=@n55, bAlphaBoolEnabled=True, ExternalForce="0,0,0") {
    alphaInputType=Bool
    angularDampingOverride=1
    angularSpringConstant=2000
    bAngularSpring=true
    bOverrideAngularDamping=true
    bUseGravityOverride=true
    bUseSphericalLimits=true
    simulationSpace=BoneRelative
    alphaBoolBlend {blendInTime=0.1, blendOption=Custom, blendOutTime=0.1}
    boundBone {boneName=dyn_strap_r3_01}
    componentAppliedLinearAccClamp=<1000 1000 1000>
    componentLinearAccScale=<0.3 0.3 0.3>
    componentLinearVelScale=<0.3 0.3 0.3>
    gravityOverride {z=-980}
    physicsBodyDefinitions [{
      boundBone {boneName=dyn_strap_r3_01}
      boxExtents=<10 5 10>
      constraintSetup {angularLimitsMax=<0 0 15>, angularLimitsMin=<0 -10 0>, angularTarget=<0 0 0>}
      localJointOffset=<1 0 0>
     }]
    relativeSpaceBone {boneName=root}
    sphericalLimits [{limitRadius=8, drivingBone {boneName=thigh_r}, sphereLocalOffset=<0 0 0>}]
   }
   n57 = Anim Dynamics - Bone: dyn_strap_r3_02(ComponentPose=@n56, bAlphaBoolEnabled=True, ExternalForce="0,0,0") {
    alphaInputType=Bool
    angularDampingOverride=1
    angularSpringConstant=2000
    bAngularSpring=true
    bOverrideAngularDamping=true
    bUseGravityOverride=true
    bUseSphericalLimits=true
    simulationSpace=BoneRelative
    alphaBoolBlend {blendInTime=0.1, blendOption=Custom, blendOutTime=0.1}
    boundBone {boneName=dyn_strap_r3_02}
    componentAppliedLinearAccClamp=<1000 1000 1000>
    componentLinearAccScale=<0.3 0.3 0.3>
    componentLinearVelScale=<0.3 0.3 0.3>
    gravityOverride {z=-980}
    physicsBodyDefinitions [{
      boundBone {boneName=dyn_strap_r3_02}
      boxExtents=<10 5 10>
      constraintSetup {angularLimitsMax=<0 0 15>, angularLimitsMin=<0 -10 0>, angularTarget=<0 0 0>}
      localJointOffset=<1 0 0>
     }]
    relativeSpaceBone {boneName=root}
    sphericalLimits [{limitRadius=8, drivingBone {boneName=thigh_r}, sphereLocalOffset=<0 0 0>}]
   }
   n58 = Anim Dynamics - Bone: dyn_strap_r2_02(ComponentPose=@n57, bAlphaBoolEnabled=True, ExternalForce="0,0,0") {
    alphaInputType=Bool
    angularDampingOverride=1
    angularSpringConstant=1000
    bAngularSpring=true
    bOverrideAngularDamping=true
    bUseGravityOverride=true
    bUsePlanarLimit=false
    bUseSphericalLimits=true
    simulationSpace=RootRelative
    alphaBoolBlend {blendInTime=0.1, blendOption=Linear, blendOutTime=0.1}
    boundBone {boneName=dyn_strap_r2_02}
    componentAppliedLinearAccClamp=<1000 1000 1000>
    componentLinearAccScale=<0.3 0.3 0.3>
    componentLinearVelScale=<0.3 0.3 0.3>
    gravityOverride {z=-980}
    physicsBodyDefinitions [{
      collisionType=CustomSphere
      sphereCollisionRadius=2.5
      boundBone {boneName=dyn_strap_r2_02}
      boxExtents=<8 4 8>
      constraintSetup {angularLimitsMax=<5 0 5>, angularLimitsMin=<-5 0 0>, angularTarget=<0 0 0>}
      localJointOffset=<1 0 0>
     }]
    relativeSpaceBone {boneName=pelvis}
   }
   Reroute Node(InputPin:Component Space Pose Link Structure=@n58) -> (OutputPin:Component Space Pose Link Structure)
  ]}
 EventGraph {nodes [
   n1 = Event BlueprintUpdateAnimation()
   > Set Delta Time X(Delta Time X=@n1.DeltaTimeX)
   > n2 = Sequence!(execute) -> (then_0, then_1, then_2:Exec)
   > Set Delta time remap(Delta time remap:double=FInterpTo(Current=Get Delta time remap() -> (Delta time remap:double), Target=1, DeltaTime=0, InterpSpeed=1)) -> (Output_Get:double)
   n3 = SpursRotation!(execute=@n2.then_1, self, Spin:double=Get Spin() -> (Spin:double), Speed:double=Get Speed() -> (Speed:double)) -> (then, Output:double)
   > n4 = Set Spin(Spin:double=@n3) -> (Output_Get:double)
   > Set Spin result(Spin result:Rotator=MakeRotator(Roll=@n4)) -> (Output_Get:Rotator)
   n5 = Get!() -> (Spin result_Roll:float, Spin result_Pitch:float, Spin result_Yaw:float)
   > Set!(
    execute
    Left Spur Rotation _Roll:float=float / float(A=@n5, B=1.2)
    Left Spur Rotation _Pitch:float=@n5.Spin result_Pitch
    Left Spur Rotation _Yaw:float=@n5.Spin result_Yaw
   ) -> (then, Output_Get:Rotator)
   n6 = Get()
   > Set!(
    execute
    Pistol Spur Rotation_Roll:float=float / float(A=@n6, B=2)
    Pistol Spur Rotation_Pitch:float=@n6.Spin result_Pitch
    Pistol Spur Rotation_Yaw:float=@n6.Spin result_Yaw
   ) -> (then, Output_Get)
   > Set NotCrouching(NotCrouching:bool=NOT Boolean(A=Get bIsCrouching())) -> (Output_Get:bool)
   > Set IsNotFrontEnd(IsNotFrontEnd:bool=NOT Boolean(A=Get bIsInFrontEnd())) -> (Output_Get:bool)
   > Set IsFrontEndAndPlayingEmote(IsFrontEndAndPlayingEmote:bool=AND Boolean(A=Get bIsInFrontEnd(), B=Get bIsPlayingEmote())) -> (Output_Get:bool)
   Set backrotator(
    execute=@n2.then_2
    backrotator:bool=OR Boolean!(A=Get bIsPlayingMeleeAnim(), B=Get bIsPlayingEmote(), C:bool=Get bIsMoving2D(), D:bool=Get bIsJumping(), E:bool=Get bIsFalling()) -> (ReturnValue)
   ) -> (Output_Get:bool)
   n7 = Event BlueprintProcessAnimInputEvent()
   > n8 = Branch(Condition=IsNamed(self=@n7.AnimInputEvent, InEventName=aie.Player.Zipline.Entered))
   > Set Zipline(Zipline:bool=true) -> (Output_Get:bool)
   Branch(execute=@n8.else, Condition=IsNamed(self=@n7.AnimInputEvent, InEventName=aie.Player.Zipline.Exited))
   > Set Zipline(Zipline=false)
  ]}
 SpursRotation {nodes [
   n1 = SpursRotation!() -> (then, Spin:double=~, Speed:double=~) {is=title K2Node_FunctionEntry}
   > n2 = Set Spin var(Spin var:double=@n1) -> (Output_Get:double)
   n3 = OR Boolean!(A=Get bIsPlayingMeleeAnim(), B=Get bIsPlayingEmote(), C=Get bIsSliding(), D=Get bIsMoving2D()) -> (ReturnValue)
   > Set Acceleration(Acceleration:double=FInterpTo(Current=Get Acceleration() -> (Acceleration:double), Target=SelectFloat(A=1.2, B=0, bPickA=@n3), DeltaTime=Get Delta Time X(), InterpSpeed=1)) -> (Output_Get:double)
   > n4 = Branch(Condition="float >= float"(A=@n2, B=720))
   > n5 = Set Spin var(Spin var=float - float(A=Get Spin var() -> (Spin var:double), B=720))
   n6 = Get Velocity mult() -> (Velocity mult:double)
   n7 = Get Acceleration()
   n8 = Get Delta time remap()
   n9 = Set Spin var(
    execute=@n4.else
    Spin var=float + float!(A=Get Spin var(), B=float * float(A=@n1.Speed, B=@n8) {is=title K2Node_PromotableOperator}, C:double=float * float(A=@n7, B=@n6)) -> (ReturnValue)
   )
   Return Node!(execute=[@n9.then, @n5.then], Output:double=@n9)
  ]}
]
variables [
 Acceleration {type=number}
 backrotator {type=boolean}
 Delta time remap {type=number}
 Delta Time X {type=number}
 FrontEndAdjustment {type=boolean}
 IsFrontEndAndPlayingEmote {type=boolean}
 IsNotFrontEnd {type=boolean}
 "Left Spur Rotation " {type=Rotator}
 NotCrouching {type=boolean}
 Pistol Spur Rotation {type=Rotator}
 Speed {type=number}
 Spin {type=number}
 Spin result {type=Rotator}
 Spin var {type=number}
 Velocity mult {type=number, default=12}
 Zipline {type=boolean}
]
