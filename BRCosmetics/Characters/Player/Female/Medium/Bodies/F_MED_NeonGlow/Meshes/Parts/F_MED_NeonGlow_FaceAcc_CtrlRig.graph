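# Control Rig for the NeonGlow female body's face accessory part that applies corrective offsets to hair bones based on the right clavicle's pose relative to the head.
# A spherical pose reader on clavicle_r, parented to head, outputs one weight.
# That weight blends chained Modify Transforms offsets onto left, back and right dyn_hair MSTR bones, the parents of the matching dyn bones.
# The dyn_hair_r2_01_MSTR offset node is outside the execution chain.
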ControlRigBlueprint /BRCosmetics/Characters/Player/Female/Medium/Bodies/F_MED_NeonGlow/Meshes/Parts/F_MED_NeonGlow_FaceAcc_CtrlRig
graphs [
 RigVMFunctionLibrary {links [], nodes []}
 RigVMModel {nodes [
   RigUnit_BeginExecution = &RigUnit_BeginExecution()
   n1 = &RigUnit_SphericalPoseReader(
    ActiveRegionScaleFactors=(PositiveWidth=1,NegativeWidth=1,PositiveHeight=1,NegativeHeight=1)
    ActiveRegionSize=0.02
    Debug=(bDrawDebug=false,bDraw2D=False,bDrawLocalAxes=False,DebugScale=25,DebugSegments=20,DebugThickness=0.25)
    DriverAxis=(X=-1,Y=0,Z=0)
    DriverItem=(Type=Bone,Name="clavicle_r")
    FalloffRegionScaleFactors=(PositiveWidth=1,NegativeWidth=0.5,PositiveHeight=1,NegativeHeight=0.5)
    FalloffSize=0.35
    FlipHeightScaling=False
    FlipWidthScaling=False
    OptionalParentItem=(Type=Bone,Name="head")
    RotationOffset=(X=195,Y=-20,Z=-35)
    ExecutePin=@RigUnit_BeginExecution.ExecutePin
   )
   > &RigUnit_ModifyTransforms(
    ItemToModify=((Item=(Type=Bone,Name="dyn_hair_l_01_MSTR"),Transform=(Rotation=(X=0,Y=0,Z=0.3031,W=-0.953),Translation=(X=-0.0834,Y=-0.1179,Z=0),Scale3D=(X=1,Y=1,Z=1))))
    WeightMinimum=0
    Weight=@n1.OutputParam
   )
   > &RigUnit_ModifyTransforms(
    ItemToModify=((Item=(Type=Bone,Name="dyn_hair_bk_l_01_MSTR"),Transform=(Rotation=(X=0.1746,Y=0.218,Z=0.2638,W=-0.9233),Translation=(X=0.249,Y=1.488,Z=1.1132),Scale3D=(X=1,Y=1,Z=1))))
    WeightMinimum=0
    Weight=@n1.OutputParam
   )
   > &RigUnit_ModifyTransforms(
    ItemToModify=((Item=(Type=Bone,Name="dyn_hair_bk_c_00_MSTR"),Transform=(Rotation=(X=0.3641,Y=0.2891,Z=0.2396,W=-0.8523),Translation=(X=0.4523,Y=4.6876,Z=-0.3182),Scale3D=(X=1,Y=1,Z=1))))
    WeightMinimum=0
    Weight=@n1.OutputParam
   )
   > &RigUnit_ModifyTransforms(
    ItemToModify=((Item=(Type=Bone,Name="dyn_hair_bk_r_00_MSTR"),Transform=(Rotation=(X=0.4461,Y=0.4179,Z=-0.0196,W=-0.7911),Translation=(X=-2.7279,Y=4.7973,Z=-2.8428),Scale3D=(X=1,Y=1,Z=1))))
    WeightMinimum=0
    Weight=@n1.OutputParam
   )
   > &RigUnit_ModifyTransforms(
    ItemToModify=((Item=(Type=Bone,Name="dyn_hair_r2_00_MSTR"),Transform=(Rotation=(X=0.4029,Y=0.4021,Z=-0.3876,W=-0.7251),Translation=(X=-0.8405,Y=0.7376,Z=-2.8021),Scale3D=(X=1,Y=1,Z=1))))
    WeightMinimum=0
    Weight=@n1.OutputParam
   )
   > &RigUnit_ModifyTransforms(
    ItemToModify=((Item=(Type=Bone,Name="dyn_hair_r1_01_MSTR"),Transform=(Rotation=(X=0,Y=0,Z=0.4353,W=0.9003),Translation=(X=0.9026,Y=4.0456,Z=-0.0001),Scale3D=(X=1,Y=1,Z=1))))
    WeightMinimum=0
    Weight=@n1.OutputParam
   )
   &RigUnit_ModifyTransforms(
    ExecutePin=()
    ItemToModify=((Item=(Type=Bone,Name="dyn_hair_r2_01_MSTR"),Transform=(Rotation=(X=0,Y=0,Z=0,W=1),Translation=(X=-3.2429,Y=-2.6936,Z=-2.5554),Scale3D=(X=1,Y=1,Z=1))))
    WeightMinimum=0
    Weight=@n1.OutputParam
   )
  ]}
]
hierarchy [
 root {loc=<0 0 0>, rot=<0 0 0 1>}
 pelvis {parent=root, loc=<0 -0.5301 98.0571>, rot=<0.0127 0.707 -0.0127 -0.707>}
 spine_01 {parent=pelvis, loc=<6.1208 -0.9328 0>, rot=<0 0 0.034 -0.9994>}
 spine_02 {parent=spine_01, loc=<6.5997 -0.2095 0>, rot=<0 0 0.0238 -0.9997>}
 spine_03 {parent=spine_02, loc=<6.4914 -0.1237 0>, rot=<0 0 0.0149 -0.9999>}
 spine_04 {parent=spine_03, loc=<6.9598 0.0794 0>, rot=<0 0 -0.0132 -0.9999>}
 spine_05 {parent=spine_04, loc=<6.9987 0 0>, rot=<0 0 -0.0015 -1>}
 neck_01 {parent=spine_05, loc=<11.3138 0.1529 0>, rot=<0 0 -0.2003 -0.9797>}
 neck_02 {parent=neck_01, loc=<4.5627 0 0>, rot=<0 0 0.0504 -0.9987>}
 head {parent=neck_02, loc=<5.3914 0 0>, rot=<0 0 0.115 -0.9934>}
 dyn_curl_fr_01_MSTR {parent=head, loc=<6.4621 9.4675 4.8658>, rot=<-0.0169 0.2589 0.9642 0.0548>}
 dyn_curl_fr_01 {parent=dyn_curl_fr_01_MSTR, loc=<0 0 0>, rot=<0 0 0 1>}
 dyn_curl_r_01_MSTR {parent=head, loc=<1.9732 6.2607 8.077>, rot=<0.111 0.3878 0.908 -0.1132>}
 dyn_curl_r_01 {parent=dyn_curl_r_01_MSTR, loc=<0 0 0>, rot=<0 0 0 1>}
 dyn_hair_r1_01_MSTR {parent=head, loc=<2.4034 4.2798 8.3908>, rot=<0.0353 0.1332 0.9905 0.0046>}
 dyn_hair_r1_01 {parent=dyn_hair_r1_01_MSTR, loc=<0 0 0>, rot=<0 0 0 1>}
 dyn_hair_r1_02_MSTR {parent=dyn_hair_r1_01, loc=<7.079 0.5012 -1.1207>, rot=<0 -0.0318 0 -0.9995>}
 dyn_hair_r1_02 {parent=dyn_hair_r1_02_MSTR, loc=<0 0 0>, rot=<0 0 0 1>}
 dyn_hair_r2_00_MSTR {parent=head, loc=<9.9411 -1.5317 8.9039>, rot=<0.0345 0.0202 0.9992 0.0085>}
 dyn_hair_r2_00 {parent=dyn_hair_r2_00_MSTR, loc=<0 0 0>, rot=<0 0 0 1>}
 dyn_hair_r2_01_MSTR {parent=dyn_hair_r2_00, loc=<7.0024 0 0>, rot=<0 0 0 1>}
 dyn_hair_r2_01 {parent=dyn_hair_r2_01_MSTR, loc=<0 0 0>, rot=<0 0 0 1>}
 dyn_hair_r2_02_MSTR {parent=dyn_hair_r2_01, loc=<7.6504 -0.62 -0.5336>, rot=<0 0 0 1>}
 dyn_hair_r2_02 {parent=dyn_hair_r2_02_MSTR, loc=<0 0 0>, rot=<0 0 0 1>}
 dyn_hair_r2_12_MSTR {parent=dyn_hair_r2_01, loc=<7.6504 3.4656 0.3551>, rot=<0 0 0 1>}
 dyn_hair_r2_12 {parent=dyn_hair_r2_12_MSTR, loc=<0 0 0>, rot=<0 0 0 1>}
 dyn_curl_r_02_MSTR {parent=dyn_hair_r2_00, loc=<6.044 -1.2898 0.8714>, rot=<0.0161 0.2006 -0.1734 -0.9641>}
 dyn_curl_r_02 {parent=dyn_curl_r_02_MSTR, loc=<0 0 0>, rot=<0 0 0 1>}
 dyn_hair_bk_r_00_MSTR {parent=head, loc=<10.1023 -6.181 5.6451>, rot=<0.0903 -0.3431 0.9337 -0.0475>}
 dyn_hair_bk_r_00 {parent=dyn_hair_bk_r_00_MSTR, loc=<0 0 0>, rot=<0 0 0 1>}
 dyn_hair_bk_r_01_MSTR {parent=dyn_hair_bk_r_00, loc=<7.005 0 -0.8355>, rot=<0 0 0 1>}
 dyn_hair_bk_r_01 {parent=dyn_hair_bk_r_01_MSTR, loc=<0 0 0>, rot=<0 0 0 1>}
 dyn_hair_bk_r_02_MSTR {parent=dyn_hair_bk_r_01, loc=<7.7512 0 0>, rot=<0 -0.0416 0 -0.9991>}
 dyn_hair_bk_r_02 {parent=dyn_hair_bk_r_02_MSTR, loc=<0 0 0>, rot=<0 0 0 1>}
 dyn_hair_bk_c_00_MSTR {parent=head, loc=<9.8408 -7.7584 0.3167>, rot=<0.0462 -0.6695 0.7396 -0.0511>}
 dyn_hair_bk_c_00 {parent=dyn_hair_bk_c_00_MSTR, loc=<0 0 0>, rot=<0 0 0 1>}
 dyn_hair_bk_c_01_MSTR {parent=dyn_hair_bk_c_00, loc=<6.8273 0 0.166>, rot=<0 0 0 1>}
 dyn_hair_bk_c_01 {parent=dyn_hair_bk_c_01_MSTR, loc=<0 0 0>, rot=<0 0 0 1>}
 dyn_hair_bk_c_02 {parent=dyn_hair_bk_c_01, loc=<6.947 0 0>, rot=<0 0 0 1>}
 dyn_hair_bk_l_01_MSTR {parent=head, loc=<3.2797 -6.8027 -4.5719>, rot=<0.0398 -0.8058 0.5885 -0.0526>}
 dyn_hair_bk_l_01 {parent=dyn_hair_bk_l_01_MSTR, loc=<0 0 0>, rot=<0 0 0 1>}
 dyn_hair_bk_l_02 {parent=dyn_hair_bk_l_01, loc=<7.5701 0 0.8402>, rot=<0 0 0 1>}
 dyn_hair_l_01_MSTR {parent=head, loc=<2.8482 -2.9685 -7.8852>, rot=<0.0239 -0.9742 0.2215 -0.0347>}
 dyn_hair_l_01 {parent=dyn_hair_l_01_MSTR, loc=<0 0 0>, rot=<0 0 0 1>}
 dyn_hair_l_02 {parent=dyn_hair_l_01, loc=<7.0191 -0.9253 1.3705>, rot=<0 0 0 1>}
 dyn_curl_l_01_MSTR {parent=dyn_hair_l_01, loc=<5.2297 1.8402 0.1764>, rot=<0.071 -0.1408 -0.3029 -0.9399>}
 dyn_curl_l_01 {parent=dyn_curl_l_01_MSTR, loc=<0 0 0>, rot=<0 0 0 1>}
 dyn_curl_l_02_MSTR {parent=dyn_hair_l_01, loc=<7.0811 0.9144 -0.5123>, rot=<0.0764 -0.1382 -0.2668 -0.9507>}
 dyn_curl_l_02 {parent=dyn_curl_l_02_MSTR, loc=<0 0 0>, rot=<0 0 0 1>}
 dyn_curl_l_03_MSTR {parent=dyn_hair_l_01, loc=<2.0256 -1.4231 2.5352>, rot=<-0.0871 0.3124 0.0396 -0.9451>}
 dyn_curl_l_03 {parent=dyn_curl_l_03_MSTR, loc=<0 0 0>, rot=<0 0 0 1>}
 dyn_earring_MSTR {parent=head, loc=<0.1678 -0.0077 -5.9225>, rot=<0.004 -0.9602 0.2794 -0.0012>}
 dyn_earring {parent=dyn_earring_MSTR, loc=<0 0 0>, rot=<0 0 0 1>}
 clavicle_l {parent=spine_05, loc=<5.4831 -0.1053 -3.2946>, rot=<0.0465 -0.7289 -0.009 -0.6829>}
 upperarm_l {parent=clavicle_l, loc=<9.8658 -1.2136 -0.4105>, rot=<0.0164 -0.3245 -0.0269 -0.9454>}
 clavicle_r {parent=spine_05, loc=<5.4831 -0.1053 3.2946>, rot=<0.7289 0.0465 -0.6829 0.009>}
 upperarm_r {parent=clavicle_r, loc=<-9.8658 1.2136 0.4105>, rot=<0.0164 -0.3245 -0.0269 -0.9454>}
]
