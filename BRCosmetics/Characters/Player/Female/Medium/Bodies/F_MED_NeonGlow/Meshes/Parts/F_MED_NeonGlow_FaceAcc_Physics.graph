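# Physics asset for the NeonGlow female medium character's FaceAcc part, setting up physics for its dynamic hair bones.
# Capsule bodies form back, left and right hair chains, each rooted at a _MSTR bone through swing-limited constraints with very tight twist.
# Extra constraints with free swing and twist cross-link neighboring lower strand bodies.
# Capsules on the neck, both clavicles and right upper arm have no constraints, apparently serving as body colliders.
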PhysicsAsset /BRCosmetics/Characters/Player/Female/Medium/Bodies/F_MED_NeonGlow/Meshes/Parts/F_MED_NeonGlow_FaceAcc_Physics
bodies {
 clavicle_l [{shapeType=Capsule, length=10, radius=6, shapeName=clavicle_l_capsule, center=<4.799 -1.5466 -2.1147>, rotation=<rot 90 180 176.8795>}]
 clavicle_r [{shapeType=Capsule, length=5, radius=7, shapeName=clavicle_r_capsule, center=<-3.0446 0.9581 1.9008>, rotation=<rot 82.6075 1.2872 -4.7676>}]
 dyn_hair_bk_c_00 [{shapeType=Capsule, length=2.5153, radius=2.6119, shapeName=dyn_hair_bk_c_00_capsule, center=<3.9869 0.0128 0.5555>, rotation=<rot 0 0 90>}]
 dyn_hair_bk_c_00_MSTR [{shapeType=Capsule, length=0.505, radius=0.505, shapeName=dyn_hair_bk_c_00_MSTR_capsule, rotation=<rot 0 -90 0>}]
 dyn_hair_bk_c_01 [{shapeType=Capsule, length=2, radius=3, shapeName=dyn_hair_bk_c_01_capsule, center=<3.6634 0 0>, rotation=<rot 90 0 0>}]
 dyn_hair_bk_c_02 [{shapeType=Capsule, length=6.5, radius=3, shapeName=dyn_hair_bk_c_02_capsule, center=<3.4001 0 0.1853>, rotation=<rot 0 0 90>}]
 dyn_hair_bk_l_01 [{shapeType=Capsule, length=2, radius=3, shapeName=dyn_hair_bk_l_01_capsule, center=<3.6634 0 0>, rotation=<rot 90 0 0>}]
 dyn_hair_bk_l_01_MSTR [{shapeType=Capsule, length=2, radius=0.5, shapeName=dyn_hair_bk_l_01_MSTR_capsule, rotation=<rot 0 0 90>}]
 dyn_hair_bk_l_02 [{shapeType=Capsule, length=3, radius=3, shapeName=dyn_hair_bk_l_02_capsule, center=<3 0 0>, rotation=<rot 90 0 0>}]
 dyn_hair_bk_r_00 [{shapeType=Capsule, length=2.5, radius=2.5, shapeName=dyn_hair_bk_r_00_capsule, center=<3.8378 0 0>, rotation=<rot 0 0 90>}]
 dyn_hair_bk_r_00_MSTR [{shapeType=Capsule, length=0.505, radius=0.505, shapeName=dyn_hair_bk_r_00_MSTR_capsule, rotation=<rot 0 -90 0>}]
 dyn_hair_bk_r_01 [{shapeType=Capsule, length=2, radius=3, shapeName=dyn_hair_bk_r_01_capsule, center=<3.6634 0 0>, rotation=<rot 90 0 0>}]
 dyn_hair_bk_r_02 [{
   shapeType=Capsule
   length=4
   radius=2.5
   shapeName=dyn_hair_bk_r_02_capsule
   center=<3.9296 0.2022 0.5769>
   rotation=<rot -0.0032 -0.0038 109.323>
  }]
 dyn_hair_l_01 [{shapeType=Capsule, length=2, radius=3, shapeName=dyn_hair_l_01_capsule, center=<3.6634 0 0>, rotation=<rot 90 0 0>}]
 dyn_hair_l_01_MSTR [{shapeType=Capsule, length=2, radius=0.5, shapeName=dyn_hair_l_01_MSTR_capsule, rotation=<rot 0 0 90>}]
 dyn_hair_l_02 [{shapeType=Capsule, length=3, radius=3, shapeName=dyn_hair_l_02_capsule, center=<3 0 0>, rotation=<rot 90 0 0>}]
 dyn_hair_r1_01 [{shapeType=Capsule, length=2, radius=3, shapeName=dyn_hair_r1_01_capsule, center=<3.6634 0 0>, rotation=<rot 90 0 0>}]
 dyn_hair_r1_01_MSTR [{shapeType=Capsule, length=2, radius=0.5, shapeName=dyn_hair_r1_01_MSTR_capsule, rotation=<rot 0 0 90>}]
 dyn_hair_r1_02 [{shapeType=Capsule, length=2.5, radius=2.5, shapeName=dyn_hair_r1_02_capsule, center=<3 0 0>, rotation=<rot 90 0 0>}]
 dyn_hair_r2_01 [{shapeType=Capsule, length=2, radius=3, shapeName=dyn_hair_r2_01_capsule, center=<3.6634 0 0>, rotation=<rot 90 0 0>}]
 dyn_hair_r2_01_MSTR [{shapeType=Capsule, length=2, radius=0.5, shapeName=dyn_hair_r2_01_MSTR_capsule, rotation=<rot 0 0 90>}]
 dyn_hair_r2_02 [{shapeType=Capsule, length=3, radius=2.5, shapeName=dyn_hair_r2_02_capsule, center=<3 0 0.8718>, rotation=<rot 90 0 0>}]
 dyn_hair_r2_12 [{
   shapeType=Capsule
   length=3
   radius=2.5
   shapeName=dyn_hair_r2_12_capsule
   center=<4.7724 -1.1299 0.9812>
   rotation=<rot 53.2197 101.9469 166.1205>
  }]
 neck_01 [{shapeType=Capsule, length=3, radius=9, shapeName=neck_01_capsule, center=<0 2.655 0>, rotation=<rot 90 0 0>}]
 upperarm_r [{
   shapeType=Capsule
   length=9.5
   radius=4.005
   shapeName=upperarm_r_capsule
   center=<-7.1373 0.5432 0.1844>
   rotation=<rot 83.7275 -0.0019 -3.3303>
  }]
}
constraints [
 {bone1Name=dyn_hair_bk_l_01, bone2Name=dyn_hair_bk_l_01_MSTR, swing1LimitDegrees=20, swing2LimitDegrees=40, twistLimitDegrees=1}
 {bone1Name=dyn_hair_l_01, bone2Name=dyn_hair_l_01_MSTR, swing1LimitDegrees=30, swing2LimitDegrees=40, twistLimitDegrees=1}
 {bone1Name=dyn_hair_r1_01, bone2Name=dyn_hair_r1_01_MSTR, swing1LimitDegrees=30, swing2LimitDegrees=30, twistLimitDegrees=1}
 {bone1Name=dyn_hair_r2_01, bone2Name=dyn_hair_r2_01_MSTR, swing1LimitDegrees=30, swing2LimitDegrees=30, twistLimitDegrees=1}
 {bone1Name=dyn_hair_r2_02, bone2Name=dyn_hair_r2_01, swing1LimitDegrees=10, swing2LimitDegrees=20, twistLimitDegrees=1}
 {bone1Name=dyn_hair_r2_12, bone2Name=dyn_hair_r2_01, swing1LimitDegrees=20, swing2LimitDegrees=30, twistLimitDegrees=1}
 {bone1Name=dyn_hair_r1_02, bone2Name=dyn_hair_r1_01, swing1LimitDegrees=20, swing2LimitDegrees=20, twistLimitDegrees=1}
 {bone1Name=dyn_hair_l_02, bone2Name=dyn_hair_l_01, swing1LimitDegrees=10, swing2LimitDegrees=20, twistLimitDegrees=1}
 {bone1Name=dyn_hair_bk_r_02, bone2Name=dyn_hair_bk_r_01, swing1LimitDegrees=1, swing2LimitDegrees=30, twistLimitDegrees=1}
 {bone1Name=dyn_hair_bk_l_02, bone2Name=dyn_hair_bk_l_01, swing1LimitDegrees=10, swing2LimitDegrees=20, twistLimitDegrees=1}
 {bone1Name=dyn_hair_bk_c_02, bone2Name=dyn_hair_bk_c_01, swing1LimitDegrees=10, swing2LimitDegrees=20, twistLimitDegrees=1}
 {
  bone1Name=dyn_hair_r2_12
  bone2Name=dyn_hair_bk_r_02
  swing1LimitDegrees=45
  swing1Motion=Free
  swing2LimitDegrees=45
  swing2Motion=Free
  twistLimitDegrees=45
  twistMotion=Free
 }
 {
  bone1Name=dyn_hair_bk_r_02
  bone2Name=dyn_hair_bk_c_02
  swing1LimitDegrees=45
  swing1Motion=Free
  swing2LimitDegrees=45
  swing2Motion=Free
  twistLimitDegrees=45
  twistMotion=Free
 }
 {bone1Name=dyn_hair_bk_c_01, bone2Name=dyn_hair_bk_c_00, swing1LimitDegrees=20, swing2LimitDegrees=50, twistLimitDegrees=1}
 {bone1Name=dyn_hair_bk_c_00, bone2Name=dyn_hair_bk_c_00_MSTR, swing1LimitDegrees=5, swing2LimitDegrees=10, twistLimitDegrees=1}
 {bone1Name=dyn_hair_bk_r_01, bone2Name=dyn_hair_bk_r_00, swing1LimitDegrees=1, swing2LimitDegrees=40, twistLimitDegrees=1}
 {bone1Name=dyn_hair_bk_r_00, bone2Name=dyn_hair_bk_r_00_MSTR, swing1LimitDegrees=5, swing2LimitDegrees=10, twistLimitDegrees=1}
 {
  bone1Name=dyn_hair_l_02
  bone2Name=dyn_hair_bk_l_02
  swing1LimitDegrees=45
  swing1Motion=Free
  swing2LimitDegrees=45
  swing2Motion=Free
  twistLimitDegrees=45
  twistMotion=Free
 }
 {
  bone1Name=dyn_hair_bk_l_02
  bone2Name=dyn_hair_bk_c_02
  swing1LimitDegrees=45
  swing1Motion=Free
  swing2LimitDegrees=45
  swing2Motion=Free
  twistLimitDegrees=45
  twistMotion=Free
 }
 {
  bone1Name=dyn_hair_r1_02
  bone2Name=dyn_hair_r2_02
  swing1LimitDegrees=45
  swing1Motion=Free
  swing2LimitDegrees=45
  swing2Motion=Free
  twistLimitDegrees=45
  twistMotion=Free
 }
]
